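; Looping gameplay cue that shows the trail effect on a Moonflax crossbow projectile, going by its name while it is in the air.
; On loop start it spawns the trail attached to the projectile's arrow mesh and, given a valid owner weapon, passes that weapon's muzzle location as MuzzlePosition.
; The trail uses manual-release pooling, so removal detaches it in world space, waits a zero-length delay, releases it to the pool and clears the reference.
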(event EventOnLoopingStart(EitherVFX-preferthisversion)
  (params (MyTarget :type Actor%20Object%20Reference) (Parameters :type Gameplay%20Cue%20Parameters%20Structure%20%28by%20ref%29) (ParticleComponents :type Array%20of%20FXSystem%20Component%20Object%20References) (AudioComponents :type Array%20of%20Audio%20Component%20Object%20References))
  (|CallParentFunction MyTarget Parameters ParticleComponents AudioComponents)
  (bind _asb_prj_new_crossbow_moonflax_athena (Utilities|Casting|CastToB_Prj_NEW_Crossbow_Moonflax_Athena MyTarget))
  (bind _returnvalue (Effects|SpawnFXSystemAttached :FXSystemAsset (Variables|Default|GetTrailFXAsset) :AttachToComponent (Variables|/MoonFlaxWeaponGameplay/Items/Weapons/Crossbow/B_Prj_NEW_Crossbow_Moonflax_Athena.B_Prj_NEW_Crossbow_Moonflax_Athena_C|GetArrowMesh _asb_prj_new_crossbow_moonflax_athena) :AttachPointName "None" :Scale "1.000000,1.000000,1.000000" :LocationType "KeepRelativeOffset" :bAutoDestroy true :PoolingMethod "ManualRelease" :bAutoActivate true))
  (bind _output_get (Variables|Default|SetTrailFX _returnvalue))
  (bind _returnvalue_2 (Class|FortProjectileBase|GetOwnerWeapon _asb_prj_new_crossbow_moonflax_athena))
  (Utilities|IsValid _returnvalue_2
    (:Is%20Valid
      (Effects|Components|ParticleSystem|SetVectorParameter _output_get "MuzzlePosition" (Effects|GetMuzzleLocation _returnvalue_2 -1)))))

(event EventOnRemoval(EitherVFX-preferthisversion)
  (params (MyTarget :type Actor%20Object%20Reference) (Parameters :type Gameplay%20Cue%20Parameters%20Structure%20%28by%20ref%29) (ParticleComponents :type Array%20of%20FXSystem%20Component%20Object%20References) (AudioComponents :type Array%20of%20Audio%20Component%20Object%20References) (BurstCameraShakeInstance :type Camera%20Shake%20Base%20Object%20Reference) (BurstDecalInstance :type Decal%20Actor%20Object%20Reference))
  (|CallParentFunction MyTarget Parameters ParticleComponents AudioComponents BurstCameraShakeInstance BurstDecalInstance)
  (bind _trailfx (Variables|Default|GetTrailFX)
    (:then
      (Transformation|DetachFromComponent _trailfx "KeepWorld" "KeepWorld" "KeepWorld" true)
      (Utilities|FlowControl|Delay 0)
      (Effects|Components|ParticleSystem|ReleasetoPool _trailfx)
      (Variables|Default|SetTrailFX 0))
    (:else)))

(fn ConstructionScript)
