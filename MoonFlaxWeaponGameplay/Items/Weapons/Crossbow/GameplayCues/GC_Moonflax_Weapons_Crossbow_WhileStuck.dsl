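; Looping gameplay cue for the Moonflax crossbow projectile while it is stuck, feeding impact data to its Niagara effects.
; On loop start it casts the target to the crossbow projectile and stores that projectile's hit normal.
; It checks the attachable projectile component's attached actor with a Pawn cast to update bPawnHit.
; It then passes bPawnHit and the impact normal as parameters to every spawned particle component.
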(event EventOnLoopingStart(Niagara)
  (params (MyTarget :type Actor%20Object%20Reference) (Parameters :type Gameplay%20Cue%20Parameters%20Structure%20%28by%20ref%29) (ParticleComponents :type Array%20of%20Niagara%20Particle%20System%20Component%20Object%20References) (AudioComponents :type Array%20of%20Audio%20Component%20Object%20References))
  ;[0.4275 0.1146 1] Determine whether the projectile hit a pawn, and set the bPawnHit Niagara parameter on any created particle components
  (|CallParentFunction MyTarget Parameters ParticleComponents AudioComponents)
  (bind _asb_prj_new_crossbow_moonflax_athena (Utilities|Casting|CastToB_Prj_NEW_Crossbow_Moonflax_Athena MyTarget))
  (Utilities|FlowControl|Sequence
    (:then_0
      (Variables|Default|SetHit%20Normal (Variables|/MoonFlaxWeaponGameplay/Items/Weapons/Crossbow/B_Prj_NEW_Crossbow_Moonflax_Athena.B_Prj_NEW_Crossbow_Moonflax_Athena_C|GetHitNormal _asb_prj_new_crossbow_moonflax_athena))
      (Utilities|IsValid (Variables|/MoonFlaxWeaponGameplay/Items/Weapons/Crossbow/B_Prj_NEW_Crossbow_Moonflax_Athena.B_Prj_NEW_Crossbow_Moonflax_Athena_C|GetFortAttachableProjectile _asb_prj_new_crossbow_moonflax_athena)
        (:Is%20Valid
          (Utilities|IsValid (Gameplay|GetAttachedtoActor (Variables|/MoonFlaxWeaponGameplay/Items/Weapons/Crossbow/B_Prj_NEW_Crossbow_Moonflax_Athena.B_Prj_NEW_Crossbow_Moonflax_Athena_C|GetFortAttachableProjectile _asb_prj_new_crossbow_moonflax_athena) :!class /Script/FortniteGame.FortAttachableProjectileComponent)
            (:Is%20Valid
              (Utilities|Casting|CastToPawn (Gameplay|GetAttachedtoActor (Variables|/MoonFlaxWeaponGameplay/Items/Weapons/Crossbow/B_Prj_NEW_Crossbow_Moonflax_Athena.B_Prj_NEW_Crossbow_Moonflax_Athena_C|GetFortAttachableProjectile _asb_prj_new_crossbow_moonflax_athena) :!class /Script/FortniteGame.FortAttachableProjectileComponent))
              (Variables|Default|SetbPawnHit false))))))
    (:then_1
      (for _array_element ParticleComponents
        (Effects|Components|ParticleSystem|SetBoolParameter _array_element "bPawnHit" (Variables|Default|GetbPawnHit))
        (Effects|Components|ParticleSystem|SetVectorParameter _array_element "Impact Normal" (Variables|Default|GetHit%20Normal))))))

(fn ConstructionScript
  (|CallParentFunction))
